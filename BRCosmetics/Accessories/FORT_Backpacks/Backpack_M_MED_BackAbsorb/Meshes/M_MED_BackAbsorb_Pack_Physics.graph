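# Physics asset for the BackAbsorb backpack cosmetic, whose bone names point to a hamburger with cheese, leaves, eyes and a flag.
# Capsule bodies form a tightly limited burger stack from base_Hamburg_01 up to dyn_Hamburg_06, with a looser joint at the base.
# Eight leaves branch from Hamburg_02, cheese pieces from Hamburg_03 (one two-link chain), and eyes and flag from Hamburg_06.
# These attachments get wider swing limits than the stack, mostly with twist locked.
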PhysicsAsset /BRCosmetics/Accessories/FORT_Backpacks/Backpack_M_MED_BackAbsorb/Meshes/M_MED_BackAbsorb_Pack_Physics
bodies {
 base_Hamburg_01 [{shapeType=Capsule, length=0.505, radius=0.505, shapeName=base_Hamburg_01_capsule, rotation=<rot 0 -90 0>}]
 dyn_Cheese_01 [{shapeType=Capsule, length=1, radius=1, center=<1.1057 0 0.0406>, rotation=<rot -87.8951 -0.0246 0.0246>}]
 dyn_Cheese_01_01 [{shapeType=Capsule, length=1, radius=1, center=<1.1057 0 0.0406>, rotation=<rot -87.8951 -0.0246 0.0246>}]
 dyn_Cheese_02_01 [{shapeType=Capsule, length=1, radius=1, center=<1.406 0 0>, rotation=<rot 90 0 0>}]
 dyn_Cheese_02_03 [{shapeType=Capsule, length=1, radius=1, center=<1.2404 0 0>, rotation=<rot 90 0 0>}]
 dyn_Hamburg_01 [{shapeType=Capsule, length=0, radius=2, shapeName=dyn_Hamburg_01_capsule, rotation=<rot 90 0 0>}]
 dyn_Hamburg_02 [{shapeType=Capsule, length=1, radius=5, shapeName=dyn_Hamburg_02_capsule, rotation=<rot -90 0 0>}]
 dyn_Hamburg_03 [{shapeType=Capsule, length=1, radius=5, shapeName=dyn_Hamburg_03_capsule, rotation=<rot -90 0 0>}]
 dyn_Hamburg_04 [{shapeType=Capsule, length=1, radius=5, shapeName=dyn_Hamburg_04_capsule, rotation=<rot -90 0 0>}]
 dyn_Hamburg_06 [{shapeType=Capsule, length=1, radius=5, shapeName=dyn_Hamburg_06_capsule, rotation=<rot -90 0 0>}]
 dyn_eye_l [{shapeType=Capsule, length=0, radius=2.5, center=<2.2591 0 -0.3483>, rotation=<rot 81.2351 0.0008 0.0008>}]
 dyn_eye_r [{shapeType=Capsule, length=0, radius=2.5, center=<2.1786 0 -0.3359>, rotation=<rot 81.2352 -0.0006 -0.0006>}]
 dyn_flag_01 [{shapeType=Capsule, length=8.7978, radius=2.3574, shapeName=dyn_flag_01_capsule, center=<6.7111 0 0>, rotation=<rot -90 0 0>}]
 dyn_leaves_01 [{shapeType=Capsule, length=3, radius=2, shapeName=dyn_leaves_01_capsule, center=<0.3503 0 0>, rotation=<rot 90 90 0>}]
 dyn_leaves_02 [{shapeType=Capsule, length=3, radius=2, shapeName=dyn_leaves_02_capsule, center=<0.3503 0 0>, rotation=<rot 90 90 0>}]
 dyn_leaves_03 [{shapeType=Capsule, length=3, radius=2, shapeName=dyn_leaves_03_capsule, center=<0.3503 0 0>, rotation=<rot 90 90 0>}]
 dyn_leaves_04 [{shapeType=Capsule, length=3, radius=2, shapeName=dyn_leaves_04_capsule, center=<0.3503 0 0>, rotation=<rot 90 90 0>}]
 dyn_leaves_05 [{shapeType=Capsule, length=3, radius=2, shapeName=dyn_leaves_05_capsule, center=<0.3503 0 0>, rotation=<rot 90 90 0>}]
 dyn_leaves_06 [{shapeType=Capsule, length=3, radius=2, shapeName=dyn_leaves_06_capsule, center=<0.3503 0 0>, rotation=<rot 90 90 0>}]
 dyn_leaves_07 [{shapeType=Capsule, length=3, radius=2, shapeName=dyn_leaves_07_capsule, center=<0.3503 0 0>, rotation=<rot 90 90 0>}]
 dyn_leaves_08 [{shapeType=Capsule, length=3, radius=2, shapeName=dyn_leaves_08_capsule, center=<0.3503 0 0>, rotation=<rot 90 90 0>}]
}
constraints [
 {bone1Name=dyn_Hamburg_01, bone2Name=base_Hamburg_01, swing1LimitDegrees=10, swing2LimitDegrees=10, twistLimitDegrees=45, twistMotion=Locked}
 {bone1Name=dyn_Hamburg_02, bone2Name=dyn_Hamburg_01, swing1LimitDegrees=0.5, swing2LimitDegrees=0.5, twistLimitDegrees=0.005}
 {bone1Name=dyn_Hamburg_03, bone2Name=dyn_Hamburg_02, swing1LimitDegrees=0.5, swing2LimitDegrees=0.5, twistLimitDegrees=0.005}
 {bone1Name=dyn_Cheese_01, bone2Name=dyn_Hamburg_03, swing1LimitDegrees=20, swing2LimitDegrees=30, twistLimitDegrees=2}
 {bone1Name=dyn_Hamburg_04, bone2Name=dyn_Hamburg_03, swing1LimitDegrees=0.5, swing2LimitDegrees=0.5, twistLimitDegrees=0.005}
 {bone1Name=dyn_leaves_01, bone2Name=dyn_Hamburg_02, swing1LimitDegrees=30, swing2LimitDegrees=30, twistLimitDegrees=1, twistMotion=Locked}
 {bone1Name=dyn_leaves_02, bone2Name=dyn_Hamburg_02, swing1LimitDegrees=30, swing2LimitDegrees=30, twistLimitDegrees=1, twistMotion=Locked}
 {bone1Name=dyn_leaves_03, bone2Name=dyn_Hamburg_02, swing1LimitDegrees=30, swing2LimitDegrees=30, twistLimitDegrees=1, twistMotion=Locked}
 {bone1Name=dyn_leaves_04, bone2Name=dyn_Hamburg_02, swing1LimitDegrees=30, swing2LimitDegrees=30, twistLimitDegrees=1, twistMotion=Locked}
 {bone1Name=dyn_leaves_05, bone2Name=dyn_Hamburg_02, swing1LimitDegrees=30, swing2LimitDegrees=30, twistLimitDegrees=1, twistMotion=Locked}
 {bone1Name=dyn_leaves_06, bone2Name=dyn_Hamburg_02, swing1LimitDegrees=30, swing2LimitDegrees=30, twistLimitDegrees=1, twistMotion=Locked}
 {bone1Name=dyn_leaves_07, bone2Name=dyn_Hamburg_02, swing1LimitDegrees=30, swing2LimitDegrees=30, twistLimitDegrees=1, twistMotion=Locked}
 {bone1Name=dyn_leaves_08, bone2Name=dyn_Hamburg_02, swing1LimitDegrees=30, swing2LimitDegrees=30, twistLimitDegrees=1, twistMotion=Locked}
 {
  bone1Name=dyn_flag_01
  bone2Name=dyn_Hamburg_06
  swing1LimitDegrees=2
  swing2LimitDegrees=2
  swing2Motion=Locked
  twistLimitDegrees=45
  twistMotion=Locked
 }
 {bone1Name=dyn_Hamburg_06, bone2Name=dyn_Hamburg_04, swing1LimitDegrees=0.5, swing2LimitDegrees=0.5, twistLimitDegrees=0.005}
 {bone1Name=dyn_eye_l, bone2Name=dyn_Hamburg_06, swing1LimitDegrees=5, swing2LimitDegrees=10, twistLimitDegrees=45, twistMotion=Locked}
 {bone1Name=dyn_eye_r, bone2Name=dyn_Hamburg_06, swing1LimitDegrees=5, swing2LimitDegrees=10, twistLimitDegrees=45, twistMotion=Locked}
 {bone1Name=dyn_Cheese_02_01, bone2Name=dyn_Hamburg_03, swing1LimitDegrees=30, swing2LimitDegrees=10, twistLimitDegrees=45, twistMotion=Locked}
 {bone1Name=dyn_Cheese_02_03, bone2Name=dyn_Hamburg_03, swing1LimitDegrees=30, swing2LimitDegrees=10, twistLimitDegrees=45, twistMotion=Locked}
 {bone1Name=dyn_Cheese_01_01, bone2Name=dyn_Cheese_01, swing1LimitDegrees=20, swing2LimitDegrees=30, twistLimitDegrees=2}
]
